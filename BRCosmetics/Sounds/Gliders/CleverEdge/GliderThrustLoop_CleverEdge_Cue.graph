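# Looping thrust sound cue for the CleverEdge glider cosmetic.
# A looping wave player feeds a continuous modulator that scales pitch and volume from the AnyDirection parameter, varying pitch slightly upward and volume within a moderate range.
# The cue uses the player glider thrust attenuation, the cosmetic glider loop properties sheet and the Fort_Gliders_Land sound class.
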SoundCue /BRCosmetics/Sounds/Gliders/CleverEdge/GliderThrustLoop_CleverEdge_Cue
root=ModulatorContinuous_1
nodes [
 {
  id=ModulatorContinuous_1
  children [WavePlayer_3]
  properties {
   pitchModulationParams {default=0, maxOutput=1.1, minOutput=1, parameterName=AnyDirection}
   volumeModulationParams {default=0, maxOutput=0.6, minOutput=0.4, parameterName=AnyDirection}
  }
 }
 {id=WavePlayer_3, properties {bLooping=true, soundWaveAssetPtr=&/GliderThrustLoop_CleverEdge_01}}
]
properties {
 attenuationSettings=&Fort_Player_Glider_Thrust_Athena_Att
 audioPropertiesSheet=&APS_Cosmetic_Glider_Loop
 maxDistance=12500
 soundClassObject=&Fort_Gliders_Land
 volumeMultiplier=0.9
 concurrencySet [null]
}
